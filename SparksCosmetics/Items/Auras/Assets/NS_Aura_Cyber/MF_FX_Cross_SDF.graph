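# Material function in the Cyber aura effect assets that builds a cross-shaped signed distance field and mask from UV coordinates.
# It calls MF_FX_Rect_SDF twice with swapped dimensions, a fixed arm width and an adjustable length "s", then unions the bars with Min.
# It outputs the raw SDF and a hard Mask made by stepping it against Mask Thickness.
# A trailing OneMinus node feeds no output.
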MaterialFunction /SparksCosmetics/Items/Auras/Assets/NS_Aura_Cyber/MF_FX_Cross_SDF
graph {nodes [
  Constant() {r=0.5}
  n1 = FunctionInput(Preview=TextureCoordinate()) {bUsePreviewValueAsDefault=true, description="st = uv", inputName=st, inputType=FunctionInput_Vector2}
  n2 = AppendVector(
   A=Constant() {r=0.25}
   B=FunctionInput(Preview=Constant() {r=1}) {bUsePreviewValueAsDefault=true, inputName=s, inputType=FunctionInput_Scalar, sortPriority=1}
  )
  n3 = ComponentMask(None=@n2) {g=true}
  n4 = ComponentMask(None=@n2) {r=true}
  n5 = Min(
   A=MaterialFunctionCall!(st=@n1, s=@n2) {materialFunction=/Game/Effects/Niagara/MasterMaterials/Functions/Mojo_MF/MF_FX_Rect_SDF}.Result
   B=MaterialFunctionCall(st=@n1, s=AppendVector(A=@n3, B=@n4)) {materialFunction=&MF_FX_Rect_SDF}.Result
  )
  FunctionOutput(None=@n5) {outputName=SDF}
  n6 = Step(
   Y=@n5
   X=FunctionInput(Preview=Constant() {r=0.5}) {bUsePreviewValueAsDefault=true, inputName=Mask Thickness, inputType=FunctionInput_Scalar, sortPriority=2}
  )
  FunctionOutput(None=@n6) {outputName=Mask, sortPriority=1}
  OneMinus(None=@n6)
 ]}
