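# Translucent, unlit Niagara sprite material that draws scrolling code for the Cyber aura.
# A Mojo scroll function samples the code texture, with per-particle random offset and stepped-scroll toggle.
# Cross SDF frame details and faint pulsing sine scanlines are added to form the mask.
# Its thresholded version hue-shifts the boosted particle color and, on mobile, replaces the soft mask multiplied by particle alpha.
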Material /SparksCosmetics/Items/Auras/Assets/NS_Aura_Cyber/M_Aura_Cyber_Code_Scroll_Sprite
graph {
 properties {
  bUsedWithNiagaraSprites=true
  blendMode=BLEND_Translucent
  pixelDepthOffsetMode=PDOM_Legacy
  shadingModel=MSM_Unlit
  twoSided=true
  usedShadingModels=Unlit
 }
 nodes [
  Constant() {desc=stepped scroll enabled, r=1}
  FeatureLevelSwitch()
  n1 = ParticleRandom()
  n2 = Multiply(A=TextureCoordinate(), B=Constant2Vector() {g=0.25, r=1})
  n3 = MaterialFunctionCall(Y=LinearInterpolate(Alpha=ParticleRandom()) {constB=100}) {materialFunction=&MakeFloat2}
  n4 = TextureSample(UVs=MaterialFunctionCall!(
    Stepped Scroll On/Off=Step(Y=@n1) {constX=0.5}
    Vertical / Horizontal Scroll=Constant()
    Speed Mult=Constant() {r=0.5}
    UV In=Add(A=@n2, B=@n3.Result)
   ) {materialFunction=/Game/Effects/Niagara/MasterMaterials/Functions/Mojo_MF/MF_FX_Screen_Scroll_Options}.UV Out) {texture=&/T_Aura_Cyber_Code_Scroll_001}
  n5 = Multiply(A=TextureCoordinate()) {constB=0.5}
  n6 = Constant() {r=0.085}
  n7 = Add(A=Multiply(A=TextureCoordinate()) {constB=0.5}, B=Constant2Vector() {g=0.5, r=0.5})
  n8 = Add(
   A=MaterialFunctionCall!(
    x=MaterialFunctionCall!(st=@n5, s, Mask Thickness) {materialFunction=&/MF_FX_Cross_SDF}.SDF
    size=@n6
   ) {materialFunction=/Game/Effects/Niagara/MasterMaterials/Functions/Mojo_MF/MF_FX_Fill}.Result
   B=MaterialFunctionCall(x=MaterialFunctionCall(st=@n7) {materialFunction=&MF_FX_Cross_SDF}.SDF, size=@n6) {materialFunction=&MF_FX_Fill}.Result
  )
  n9 = Add(
   A=TextureCoordinate()
   B=Multiply(A=Time(), B=ScalarParameter() {defaultValue=-0.25, group=Screen Detail Settings, parameterName=Scanline Y Speed, sortPriority=3}) {constB=-0.5, desc=Speed}
  )
  n10 = Multiply(
   A=ComponentMask(None=Frac(None=@n9)) {g=true}
   B=ScalarParameter() {defaultValue=12, group=Screen Detail Settings, parameterName=Scanline Distort Y Tile, sortPriority=2}
  ) {constB=5, desc=Y tile}
  n11 = Add(A=ConstantBiasScale(None=Sine(None=@n10))) {constB=0.001}
  n12 = Multiply(A=Time()) {constB=0.5}
  n13 = Add(A=Constant() {r=0.01}, B=Multiply(A=Multiply(A=@n11) {constB=0.005}, B=Sine(None=@n12)))
  n14 = Saturate(None=Add(A=Multiply(A=@n4.R) {constB=5}, B=Add(A=@n8, B=@n13) {constB=0.01}) {desc=Apply Screen Frame Details})
  n15 = Step(X=@n14) {constY=0.1}
  n16 = ScalarParameter() {defaultValue=2, parameterName=Hue Shift Mult}
  n17 = ParticleColor()
  Multiply(A=MaterialFunctionCall(Hue Shift Percentage=Multiply(A=@n15, B=@n16), Texture=@n17.RGB) {materialFunction=&HueShift}.Result) {constB=3, desc=Boost}
  OneMinus(None=@n8)
  n18 = Time() {bOverride_Period=true, period=100}
  n19 = Multiply(A=@n18, B=ScalarParameter() {defaultValue=0.62, parameterName=TimeScaleA})
  n20 = Multiply(A=@n18, B=ScalarParameter() {defaultValue=2.233, parameterName=TimeScaleB})
  Panner(Coordinate=TextureCoordinate() {vTiling=0.25}, Time=Add(A=Ceil(None=@n19), B=Ceil(None=@n20))) {bFractionalPart=true, speedY=0.02}
  ShadingPathSwitch(Default=Multiply(A=@n17.A, B=@n14), Mobile=Multiply(A=@n17.A, B=@n15) {desc=masked})
 ]
}
