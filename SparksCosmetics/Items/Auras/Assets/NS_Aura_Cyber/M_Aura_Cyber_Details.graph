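# Translucent unlit Niagara sprite material that draws screen-detail shapes for the Cyber aura cosmetic.
# SDF functions build two procedural masks: a stroked rounded box with a filled thin-box element, and a pair of filled crosses.
# The cross mask gains faint scrolling sine scanlines modulated over time.
# A step on the Mask Select dynamic parameter picks the mask per particle, and particle alpha scales the result.
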Material /SparksCosmetics/Items/Auras/Assets/NS_Aura_Cyber/M_Aura_Cyber_Details
graph {
 properties {bUsedWithNiagaraSprites=true, blendMode=BLEND_Translucent, pixelDepthOffsetMode=PDOM_Legacy, shadingModel=MSM_Unlit, usedShadingModels=Unlit}
 nodes [
  n1 = Add(A=Multiply(A=TextureCoordinate()) {constB=8}) {constB=-4}
  n2 = Constant2Vector() {g=1.25}
  n3 = MaterialFunctionCall!(
   X=MaterialFunctionCall!(HalfSize=Constant2Vector() {g=2, r=2}, CornerRadius=Constant() {r=0.02}, Position=Add(A=@n1, B=@n2)) {materialFunction=/Game/Effects/Niagara/MasterMaterials/Functions/Mojo_MF/MF_FX_RoundedBox_SDF}.Result
   S
   W=Constant() {r=0.09}
  ) {materialFunction=/Game/Effects/Niagara/MasterMaterials/Functions/Mojo_MF/MF_FX_Stroke}
  Multiply(A=@n3.Result)
  n4 = Multiply(A=TextureCoordinate()) {constB=8}
  n5 = Constant2Vector() {g=-2, r=-4}
  n6 = MaterialFunctionCall(HalfSize=Constant2Vector() {g=-0.008, r=1.75}, CornerRadius=Constant() {r=0.0015}, Position=Add(A=Add(A=@n4, B=@n5) {constB=-0.3}, B=@n2)) {materialFunction=&MF_FX_RoundedBox_SDF}
  n7 = Add(
   A=@n3.Result
   B=MaterialFunctionCall!(x=Multiply(A=@n6.Result One Minus), size=Constant() {r=0.25}) {materialFunction=/Game/Effects/Niagara/MasterMaterials/Functions/Mojo_MF/MF_FX_Fill}.Result
  )
  n8 = DynamicParameter() {defaultValue=<rgba 0.6 1 1 1>, paramNames [Mask Select, Param2, Param3, Param4]}
  n9 = Multiply(A=TextureCoordinate()) {constB=0.5}
  n10 = Constant() {r=0.085}
  n11 = Add(A=Multiply(A=TextureCoordinate()) {constB=0.5}, B=Constant2Vector() {g=0.5, r=0.5})
  n12 = Add(
   A=MaterialFunctionCall(
    x=MaterialFunctionCall!(st=@n9, s, Mask Thickness) {materialFunction=&/MF_FX_Cross_SDF}.SDF
    size=@n10
   ) {materialFunction=&MF_FX_Fill}.Result
   B=MaterialFunctionCall(x=MaterialFunctionCall(st=@n11) {materialFunction=&MF_FX_Cross_SDF}.SDF, size=@n10) {materialFunction=&MF_FX_Fill}.Result
  )
  n13 = Add(
   A=TextureCoordinate()
   B=Multiply(A=Time(), B=ScalarParameter() {defaultValue=-0.25, group=Screen Detail Settings, parameterName=Scanline Y Speed, sortPriority=3}) {constB=-0.5, desc=Speed}
  )
  n14 = Multiply(
   A=ComponentMask(None=Frac(None=@n13)) {g=true}
   B=ScalarParameter() {defaultValue=12, group=Screen Detail Settings, parameterName=Scanline Distort Y Tile, sortPriority=2}
  ) {constB=5, desc=Y tile}
  n15 = Add(A=ConstantBiasScale(None=Sine(None=@n14))) {constB=0.001}
  n16 = Multiply(A=Time()) {constB=0.5}
  n17 = Add(A=Constant() {r=0.005}, B=Multiply(A=Multiply(A=@n15) {constB=0.005}, B=Sine(None=@n16)))
  Multiply(
   A=ParticleColor().A
   B=LinearInterpolate(A=Add(A=@n12, B=@n17) {constB=0.01}, B=Saturate(None=@n7), Alpha=Step(Y=@n8.Mask Select) {constX=0.5})
  )
  OneMinus(None=@n12)
 ]
}
