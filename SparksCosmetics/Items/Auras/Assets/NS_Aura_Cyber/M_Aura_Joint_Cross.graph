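# Unlit translucent Niagara sprite material that draws a cross shape, named for the joints of the Cyber aura cosmetic.
# The MF_FX_Cross_SDF function builds the cross mask procedurally, with thickness driven per particle by the Cross Thickness dynamic parameter.
# The mask is multiplied by particle alpha and saturated.
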Material /SparksCosmetics/Items/Auras/Assets/NS_Aura_Cyber/M_Aura_Joint_Cross
graph {
 properties {bUsedWithNiagaraSprites=true, blendMode=BLEND_Translucent, pixelDepthOffsetMode=PDOM_Legacy, shadingModel=MSM_Unlit, usedShadingModels=Unlit}
 nodes [
  n1 = DynamicParameter() {defaultValue=<rgba 0.5 1 1 1>, paramNames [Cross Thickness, Param2, Param3, Param4]}
  Saturate(None=Multiply(
    A=ParticleColor().A
    B=MaterialFunctionCall!(st, s, Mask Thickness=@n1.Cross Thickness) {materialFunction=&/MF_FX_Cross_SDF}.Mask
   ))
 ]
}
