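# Translucent unlit Niagara sprite material that draws binary digits for the Cyber aura cosmetic.
# A FlipBook function samples a shared numbers texture as a single row of ten columns.
# Particle random lerps the phase within the first fifth of the strip, so each sprite picks one of the first two frames, presumably 0 or 1.
# The sampled result is multiplied by particle color alpha.
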Material /SparksCosmetics/Items/Auras/Assets/NS_Aura_Cyber/M_Binary_Sprites
graph {
 properties {bUsedWithNiagaraSprites=true, blendMode=BLEND_Translucent, pixelDepthOffsetMode=PDOM_Legacy, shadingModel=MSM_Unlit, usedShadingModels=Unlit}
 nodes [
  n1 = ParticleRandom()
  Multiply(
   A=ParticleColor().A
   B=MaterialFunctionCall(
    "Animation  Phase (0-1)"=LinearInterpolate(Alpha=@n1) {constB=0.2}
    Number of Rows=Constant() {r=1}
    Number of Columns=Constant() {r=10}
    Texture=TextureObject() {texture=/Game/Environments/Props/Street/Textures/numbers}
   ) {materialFunction=&FlipBook}.Result
  )
 ]
}
