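# MF_FX_Atlas_UV is an FX material function that computes UVs and blend weights for crossfading neighboring cells of a texture atlas, flipbook-style.
# The fractional part of Selector, scaled by Size's cell count, gives a row-major frame index.
# That index and the following one, not explicitly wrapped, become grid offsets that remap UV into two cells.
# The index's fraction yields Blend2, and its complement Blend1; sampling happens outside.
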MaterialFunction /Game/Effects/Niagara/MasterMaterials/Functions/Mojo_MF/MF_FX_Atlas_UV
graph {nodes [
  n1 = FunctionInput(Preview=Constant2Vector() {desc=size, g=2, r=2}) {bUsePreviewValueAsDefault=true, inputName=Size, inputType=FunctionInput_Vector2}
  n2 = ComponentMask(None=@n1) {desc=width, r=true}
  n3 = Constant() {desc=Selector}
  n4 = Multiply(
   A=Multiply(A=@n2, B=ComponentMask(None=@n1) {desc=height, r=true}) {desc=area}
   B=Frac(None=FunctionInput(Preview=@n3) {bUsePreviewValueAsDefault=true, inputName=Selector, inputType=FunctionInput_Scalar, sortPriority=1}) {desc=sel}
  ) {desc=findex}
  n5 = Floor(None=@n4) {desc=index1}
  n6 = Floor(None=Divide(A=@n5, B=@n2)) {desc=row1}
  n7 = MaterialFunctionCall(X=Subtract(A=@n5, B=Multiply(A=@n6, B=@n2)) {desc=col1}, Y=@n6) {desc=gridPos1, materialFunction=&MakeFloat2}
  n8 = FunctionInput(Preview=TextureCoordinate()) {bUsePreviewValueAsDefault=true, inputName=UV, inputType=FunctionInput_Vector2, sortPriority=2}
  FunctionOutput(None=Divide(A=Add(A=@n7.Result, B=@n8), B=@n1) {desc=New UV 1}) {outputName=New UV 1}
  n9 = Add(A=@n5) {desc=index2}
  n10 = Floor(None=Divide(A=@n9, B=@n2)) {desc=row2}
  n11 = MaterialFunctionCall(X=Subtract(A=@n9, B=Multiply(A=@n10, B=@n2)) {desc=col2}, Y=@n10) {desc=gridPos2, materialFunction=&MakeFloat2}
  FunctionOutput(None=Divide(A=Add(A=@n11.Result, B=@n8), B=@n1) {desc=New UV 2}) {outputName=New UV 2, sortPriority=1}
  n12 = Subtract(A=@n4, B=@n5) {desc=Blend2}
  FunctionOutput(None=OneMinus(None=@n12) {desc=Blend1}) {outputName=Blend1, sortPriority=2}
  FunctionOutput(None=@n12) {outputName=Blend2, sortPriority=3}
 ]}
