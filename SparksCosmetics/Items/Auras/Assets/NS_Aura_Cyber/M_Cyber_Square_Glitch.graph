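# Translucent unlit Niagara sprite material for square glitch particles in the Cyber aura cosmetic.
# Scrolling sine scanlines, with speed and tiling parameters, are multiplied by digital noise from a 2x2 atlas whose cell selector loops over time.
# That glitch pattern is shaped by a sine-based UV mask that fades toward the sprite edges and by particle alpha.
# Particle color feeds a shared emissive-adjustment function.
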Material /SparksCosmetics/Items/Auras/Assets/NS_Aura_Cyber/M_Cyber_Square_Glitch
graph {
 properties {bUsedWithNiagaraSprites=true, blendMode=BLEND_Translucent, pixelDepthOffsetMode=PDOM_Legacy, shadingModel=MSM_Unlit, usedShadingModels=Unlit}
 nodes [
  n1 = Time()
  n2 = ScalarParameter() {defaultValue=-0.25, group=Screen Detail Settings, parameterName=Scanline Y Speed, sortPriority=3}
  n3 = Frac(None=Add(A=TextureCoordinate(), B=Multiply(A=@n1, B=@n2) {constB=-0.5, desc=Speed}))
  n4 = Sine(None=Multiply(
    A=ComponentMask(None=@n3) {g=true}
    B=ScalarParameter() {defaultValue=2, group=Screen Detail Settings, parameterName=Scanline Distort Y Tile, sortPriority=2}
   ) {constB=5, desc=Y tile})
  ConstantBiasScale(None=@n4)
  Constant() {desc=Selector}
  n5 = ParticleColor()
  MaterialFunctionCall(Emissive=@n5.RGB) {materialFunction=&MF_FX_EmissiveAdjustments}
  n6 = Constant() {desc=tile, r=1}
  n7 = Constant() {desc=offset}
  n8 = Add(A=Multiply(A=TextureCoordinate(), B=AppendVector(A=@n6, B=@n6)), B=AppendVector(A=@n7, B=@n7))
  n9 = Fmod(
   A=Multiply(A=Time(), B=ScalarParameter() {defaultValue=5, group=Distortion Controls, parameterName=Scanline Distortion Speed Mult}) {constB=5}
   B=Constant() {r=1}
  )
  n10 = TextureSample(
   UVs=MaterialFunctionCall!(Size=Constant2Vector() {desc=size, g=2, r=2}, Selector=Saturate(None=@n9), UV=Add(A=@n8)) {materialFunction=/Game/Effects/Niagara/MasterMaterials/Functions/Mojo_MF/MF_FX_Atlas_UV}.New UV 1
   Tex=TextureObject() {samplerType=SAMPLERTYPE_LinearColor, texture=/Game/Global/Textures/Default/Noise/T_Atlas_DigitalNoise_01}
  )
  Multiply(A=@n10.RGB)
  n11 = TextureCoordinate()
  n12 = Constant() {r=0.5}
  n13 = Multiply(A=Sine(None=ComponentMask(None=@n11) {r=true}) {period=2}, B=@n12) {constB=5}
  n14 = Multiply(A=Sine(None=ComponentMask(None=@n11) {g=true}) {period=2}, B=@n12) {constB=5}
  n15 = Multiply(A=@n13, B=@n14)
  n16 = Multiply(
   A=Subtract(A=ConstantBiasScale(None=@n4) {bias=0.5, scale=1}) {constB=0.99}
   B=MaterialFunctionCall(X=@n10.G, Y=@n10.G) {materialFunction=&MakeFloat2}.Result
  ) {constB=0.5}
  Saturate(None=Multiply(A=Multiply(A=@n5.A, B=@n15), B=Add(A=@n16) {constB=0.5}))
  Saturate(None=@n13)
  Saturate(None=@n14)
 ]
}
